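# Ambient sound cue for a small running motor in the Apollo AccessHub biome, with "AccessPoint" in its name.
# A Random node picks one of three motor-run wave variants with equal weight.
# It routes through the indoor Athena ambient sound class with shared ambient attenuation and plays at reduced volume.
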SoundCue /Game/Sounds/Apollo/Biomes/AccessHub/AMB_AccessPoint_MotorRunSml
root=Random_0
nodes [
 {id=Random_0, children [WavePlayer_0,WavePlayer_1,WavePlayer_3], properties {weights [1,1,1]}}
 {id=WavePlayer_0, properties {soundWaveAssetPtr=&/Biome_Access_Point_MotorRunSml_01}}
 {id=WavePlayer_1, properties {soundWaveAssetPtr=&/Biome_Access_Point_MotorRunSML_02}}
 {id=WavePlayer_3, properties {soundWaveAssetPtr=&/Biome_Access_Point_MotorRunSML_03}}
]
properties {
 attenuationSettings=&Fort_Ambient_Athena_Att
 maxDistance=2097152
 soundClassObject=&Fort_Ambient_Indoors_Athena
 volumeMultiplier=0.301
 preEffectBusSends [{}]
}
