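# Looping ambient bed SoundCue for the Access Point in the Apollo AccessHub biome.
# A Mixer layers two looping waves: the Access Point biome loop at full input volume and a quieter ominous bed.
# The cue plays at reduced volume through the Athena indoor ambient sound class, using the shared Athena ambient attenuation.
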SoundCue /Game/Sounds/Apollo/Biomes/AccessHub/AMB_Bed_AccessPoint_LP
root=Mixer_0
nodes [
 {id=Mixer_0, children [WavePlayer_2,WavePlayer_0], properties {inputVolume [0.4,1]}}
 {id=WavePlayer_0, properties {bLooping=true, soundWaveAssetPtr=&/AMB_Biome_Access_Point_mxLCH_LP_01}}
 {id=WavePlayer_2, properties {bLooping=true, soundWaveAssetPtr=&/AMB_Bed_AccessPoint_Ominous_LP}}
]
properties {
 attenuationSettings=&Fort_Ambient_Athena_Att
 maxDistance=2097152
 soundClassObject=&Fort_Ambient_Indoors_Athena
 volumeMultiplier=0.401
 preEffectBusSends [{}]
}
